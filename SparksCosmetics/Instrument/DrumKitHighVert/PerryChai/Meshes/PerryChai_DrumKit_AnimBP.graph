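# Animation Blueprint for the PerryChai drum kit cosmetic, adding procedural motion to a bear and a spinning fan.
# Five chained Anim Dynamics nodes give the bear's neck, arms and legs sprung, angle-limited secondary motion before an additive bone-space roll turns the fan bone.
# Each update accumulates a wrapped fan angle from speed and delta time.
# Front end sets speed directly; elsewhere it interpolates toward the same target, clamped.
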AnimBlueprint /SparksCosmetics/Instrument/DrumKitHighVert/PerryChai/Meshes/PerryChai_DrumKit_AnimBP
target_skeleton=&/PerryChai_DrumKit_Skeleton
graphs [
 AnimGraph {nodes [
   n1 = Anim Dynamics - Bone: dyn_bear_neck(ComponentPose=Local To Component(LocalPose=Input Pose()), Alpha=1, ExternalForce="0,0,0") {
    angularDampingOverride=0.99
    angularSpringConstant=300
    bAngularSpring=true
    bOverrideAngularDamping=true
    bUseGravityOverride=true
    simulationSpace=BoneRelative
    alphaBoolBlend {blendOption=Linear}
    boundBone {boneName=dyn_bear_neck}
    componentAppliedLinearAccClamp=<2000 2000 2000>
    componentLinearAccScale=<0.1 0.1 0.4>
    componentLinearVelScale=<1.1 1.1 1.1>
    gravityOverride {z=-350}
    physicsBodyDefinitions [{
      boundBone {boneName=dyn_bear_neck}
      boxExtents=<4 4 4>
      constraintSetup {angularLimitsMax=<5 5 20>, angularLimitsMin=<-5 -6 -20>, angularTarget=<1 0 0>}
      localJointOffset=<3 0 0>
     }]
    simSpaceSettings {maxAngularAcceleration=1000, maxAngularVelocity=100}
   }
   n2 = Anim Dynamics - Bone: dyn_bear_left_arm(ComponentPose=@n1, Alpha=1, ExternalForce="0,0,0") {
    angularDampingOverride=0.99
    angularSpringConstant=300
    bAngularSpring=true
    bOverrideAngularDamping=true
    bUseGravityOverride=true
    simulationSpace=BoneRelative
    alphaBoolBlend {blendOption=Linear}
    boundBone {boneName=dyn_bear_left_arm}
    componentAppliedLinearAccClamp=<3500 3500 3500>
    componentLinearAccScale=<0.4 0.4 0.4>
    componentLinearVelScale=<1 1 1>
    gravityOverride {z=-200}
    physicsBodyDefinitions [{
      boundBone {boneName=dyn_bear_left_arm}
      boxExtents=<10 10 10>
      constraintSetup {coneAngle=1, angularLimitsMax=<5 10 0>, angularLimitsMin=<-5 -10 -45>, angularTarget=<1 0 0>}
      localJointOffset=<1 0 0>
     }]
    simSpaceSettings {maxAngularAcceleration=1000, maxAngularVelocity=100}
   }
   n3 = Anim Dynamics - Bone: dyn_bear_right_arm(ComponentPose=@n2, Alpha=1, ExternalForce="0,0,0") {
    angularDampingOverride=0.99
    angularSpringConstant=300
    bAngularSpring=true
    bOverrideAngularDamping=true
    bUseGravityOverride=true
    simulationSpace=BoneRelative
    alphaBoolBlend {blendOption=Linear}
    boundBone {boneName=dyn_bear_right_arm}
    componentAppliedLinearAccClamp=<3500 3500 3500>
    componentLinearAccScale=<0.4 0.4 0.4>
    componentLinearVelScale=<1 1 1>
    gravityOverride {z=-200}
    physicsBodyDefinitions [{
      boundBone {boneName=dyn_bear_right_arm}
      boxExtents=<10 10 10>
      constraintSetup {coneAngle=1, angularLimitsMax=<5 10 45>, angularLimitsMin=<-5 -10 0>, angularTarget=<1 0 0>}
      localJointOffset=<1 0 0>
     }]
    simSpaceSettings {maxAngularAcceleration=1000, maxAngularVelocity=100}
   }
   n4 = Anim Dynamics - Bone: dyn_bear_leg_left(ComponentPose=@n3, Alpha=1, ExternalForce="0,0,0") {
    angularDampingOverride=0.99
    angularSpringConstant=300
    bAngularSpring=true
    bOverrideAngularDamping=true
    bUseGravityOverride=true
    simulationSpace=BoneRelative
    alphaBoolBlend {blendOption=Linear}
    boundBone {boneName=dyn_bear_leg_left}
    componentAppliedLinearAccClamp=<3500 3500 3500>
    componentLinearAccScale=<0.4 0.4 0.4>
    componentLinearVelScale=<1 1 1>
    gravityOverride {z=-300}
    physicsBodyDefinitions [{
      boundBone {boneName=dyn_bear_leg_left}
      boxExtents=<11 11 11>
      constraintSetup {coneAngle=1, angularLimitsMax=<0 0 10>, angularLimitsMin=<0 -30 -10>, angularTarget=<1 0 0>}
      localJointOffset=<1 0 0>
     }]
    simSpaceSettings {maxAngularAcceleration=1000, maxAngularVelocity=100}
   }
   n5 = Anim Dynamics - Bone: dyn_bear_leg_right(ComponentPose=@n4, Alpha=1, ExternalForce="0,0,0") {
    angularDampingOverride=0.99
    angularSpringConstant=300
    bAngularSpring=true
    bOverrideAngularDamping=true
    bUseGravityOverride=true
    simulationSpace=BoneRelative
    alphaBoolBlend {blendOption=Linear}
    boundBone {boneName=dyn_bear_leg_right}
    componentAppliedLinearAccClamp=<3500 3500 3500>
    componentLinearAccScale=<0.4 0.4 0.4>
    componentLinearVelScale=<1 1 1>
    gravityOverride {z=-300}
    physicsBodyDefinitions [{
      boundBone {boneName=dyn_bear_leg_right}
      boxExtents=<11 11 11>
      constraintSetup {coneAngle=1, angularLimitsMax=<0 0 10>, angularLimitsMin=<0 -30 -10>, angularTarget=<1 0 0>}
      localJointOffset=<1 0 0>
     }]
    simSpaceSettings {maxAngularAcceleration=1000, maxAngularVelocity=100}
   }
   n6 = "Transform (Modify) Bone - Bone: dyn_fan"!(
    ComponentPose=@n5
    Alpha=1
    Translation="0,0,0"
    Rotation_Roll:float=Get!() -> (rotation of fan_Roll:float, rotation of fan_Pitch:float, rotation of fan_Yaw:float)
    Rotation_Pitch:float=0
    Rotation_Yaw:float=0
    Scale="1,1,1"
   ) -> (Pose) {
    rotationMode=BMM_Additive
    rotationSpace=BCS_BoneSpace
    translationSpace=BCS_BoneSpace
    alphaBoolBlend {blendOption=Linear}
    boneToModify {boneName=dyn_fan}
   }
   Output Pose(Result=Component To Local(ComponentPose=@n6))
  ]}
 EventGraph {nodes [
   Event BlueprintUpdateAnimation()
   > n1 = Sequence()
   > n2 = Branch(Condition=Get bIsInFrontEnd())
   > Set RotationSpeed(RotationSpeed:double=400) -> (Output_Get:double)
   n3 = GetWorldDeltaSeconds!() -> (ReturnValue:double)
   n4 = Get RotationSpeed() -> (RotationSpeed:double)
   Set RotationSpeed(execute=@n2.else, RotationSpeed="Clamp (Float)"(Value=FInterpTo(Current=@n4, Target=400, DeltaTime=@n3, InterpSpeed=1), Min=200, Max=400))
   n5 = GetWorldDeltaSeconds()
   n6 = Get RotationSpeed()
   n7 = Set CurrentAngle(
    execute=@n1.then_1
    CurrentAngle:double=float + float(A=Get CurrentAngle() -> (CurrentAngle:double), B=float * float(A=@n5, B=@n6) {is=title K2Node_PromotableOperator})
   ) -> (Output_Get:double)
   > n8 = Set CurrentAngle(CurrentAngle="% (Float)"!(A:double=@n7, B:double=360) -> (ReturnValue:double))
   > Set rotation of fan(rotation of fan:Rotator=MakeRotator(Roll=@n8)) -> (Output_Get:Rotator)
   TryGetPawnOwner()
  ]}
]
variables [
 CurrentAngle {type=number}
 rotation of fan {type=Rotator}
 RotationSpeed {type=number}
]
